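# Masked, two-sided foliage material for front-end lobby grass, with stylised colour variation and wind sway.
# The lobby collection's FloorColor is darkened by world-XY quantised Perlin noise and a landscape grass overlay, then scaled by a grass mask channel.
# Two summed sines, phase-shifted by instance world position, drive XY sway weighted by a local-height gradient in the GrassAnim reroute.
# Material output connections are not shown.
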Material /Game/NeverCook/FX/FrontEndFX/S38/Grass/M_WE_Grass_frontend
graph {
 properties {bAutomaticallySetUsageInEditor=false, blendMode=BLEND_Masked, twoSided=true, usedShadingModels=Default Lit}
 nodes [
  Constant()
  Constant() {r=1}
  n1 = WorldPosition()
  n2 = ScalarParameter() {defaultValue=462.7858, parameterName=StylisedNoiseTiling2}
  n3 = TextureSample(UVs=Divide(A=@n1.XY, B=@n2)) {texture=/Game/Environments/FrontEnd/Materials/WildEstate_Landscape_Grass_Overlay_D}
  n4 = LinearInterpolate(A=ScalarParameter() {defaultValue=0.8876, parameterName=MinDarkness2}, Alpha=@n3.G)
  Multiply(A=@n4, B=@n3.R)
  n5 = ScalarParameter() {defaultValue=1823.9929, parameterName=StylisedNoiseTiling}
  Multiply(A=TextureCoordinate(), B=@n5)
  Multiply(A=TextureCoordinate(), B=@n2)
  n6 = ComponentMask(None=CollectionParameter() {collection=/Game/Environments/FrontEnd/BackPlates/Materials/MPC_LobbyBackground, parameterName=FloorColor}) {b=true, g=true, r=true}
  n7 = Divide(A=@n1.XY, B=@n5)
  n8 = ScalarParameter() {defaultValue=4, parameterName=Quantize}
  n9 = Round(None=Multiply(
    A=TextureSample(UVs=@n7) {samplerType=SAMPLERTYPE_LinearGrayscale, texture=&T_SimplePerlin_Greyscale}.R
    B=@n8
   ))
  n10 = Multiply(A=@n4, B=LinearInterpolate(A=ScalarParameter() {defaultValue=0.652, parameterName=MinDarkness}, Alpha=Divide(A=@n9, B=@n8)))
  Multiply(
   A=ScalarParameter() {parameterName=StylisedBaseColorContribution}
   B=Multiply(
    A=TextureSample() {samplerType=SAMPLERTYPE_Masks, texture=&/T_WIldEstate_Grass_Standard_M}.G
    B=Multiply(A=@n6, B=@n10)
   )
  )
  MaterialExpressionTransformPosition_0 = TransformPosition(None=Constant3Vector()) {transformSourceType=TRANSFORMPOSSOURCE_Instance, transformType=TRANSFORMPOSSOURCE_World}
  n11 = Convert(None=@MaterialExpressionTransformPosition_0) {
   nodeName=Break Float3
   convertInputs [{
     type=Vector3
     expressionInput {
      expression=&self:MaterialExpressionTransformPosition_0
      inputName=None
      mask=0
      maskA=0
      maskB=0
      maskG=0
      maskR=0
      outputIndex=0
     }
    }]
   convertOutputs [{type=Scalar}, {type=Scalar}, {type=Scalar}]
  }
  n12 = Add(A=Add(A=@n11, B=@n11), B=@n11)
  n13 = Multiply(A=Frac(None=Divide(A=@n12) {constB=33.33}), B=Constant() {r=33.33})
  n14 = Add(A=Time(), B=@n13)
  n15 = ScalarParameter() {defaultValue=0.5257, group=Animation, parameterName=Anim_Speed}
  n16 = Multiply(A=@n14, B=@n15)
  n17 = Multiply(A=@n14, B=Add(A=Multiply(A=@n13) {constB=0.01}, B=@n15))
  n18 = Multiply(A=Add(A=Sine(None=@n16), B=Sine(None=@n17)), B=Constant3Vector() {constant=<rgba 1 1 0 0>})
  n19 = ScalarParameter() {defaultValue=1, group=Animation, parameterName=Wind_Intensity}
  n20 = LocalPosition() {localOrigin=InstancePreSkinning}
  n21 = ScalarParameter() {defaultValue=3.5322, group=Animation, parameterName=Gradient_Bias}
  n22 = Power(Base=Divide(A=Subtract(A=@n20.Z, B=@n21), B=ScalarParameter() {defaultValue=16.6253, group=Animation, parameterName=Gradient_Dist}))
  NamedRerouteDeclaration(None=Multiply(A=Multiply(A=@n18, B=@n19), B=Saturate(None=@n22))) {name=GrassAnim, nodeColor=<rgba 0.9688 1 0 1>}
  NamedRerouteUsage()
  PerInstanceRandom()
 ]
}
